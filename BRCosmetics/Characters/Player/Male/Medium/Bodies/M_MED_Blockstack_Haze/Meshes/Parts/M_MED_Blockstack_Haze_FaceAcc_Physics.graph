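# Physics asset for the face accessory mesh of the Blockstack Haze male medium player cosmetic, defining bodies for three hair chains.
# Center, left and right chains each link a small base capsule to the head, followed by two dynamic capsules with narrow swing limits and locked twist.
# Boxes cover the head and upper spine, with a back-of-chest capsule joined to the spine.
# Base hair and chest-back joints leave swing and twist free.
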PhysicsAsset /BRCosmetics/Characters/Player/Male/Medium/Bodies/M_MED_Blockstack_Haze/Meshes/Parts/M_MED_Blockstack_Haze_FaceAcc_Physics
bodies {
 base_hair_c [{shapeType=Capsule, length=0.505, radius=0.505, shapeName=base_hair_c_capsule, rotation=<rot 0 -90 0>}]
 base_hair_l [{shapeType=Capsule, length=0.505, radius=0.505, shapeName=base_hair_l_capsule, rotation=<rot 0 -90 0>}]
 base_hair_r [{shapeType=Capsule, length=0.505, radius=0.505, shapeName=base_hair_r_capsule, rotation=<rot 0 -90 0>}]
 chest_back [{shapeType=Capsule, length=13.7274, radius=6.2726, shapeName=chest_back_capsule, center=<0 9.9416 -12.9971>, rotation=<rot 90 0 0>}]
 deform_head [{shapeType=Box, extentX=27.7804, extentY=20, extentZ=41.1091, shapeName=deform_head_box, center=<-9.4334 -1.8265 0>}]
 deform_spine_05 [{shapeType=Box, extentX=34.2204, extentY=20, extentZ=41.1091, shapeName=deform_head_box, center=<5.8908 0.2129 0>}]
 dyn_hair_01_c [{shapeType=Capsule, length=6, radius=3, shapeName=dyn_hair_01_c_capsule, center=<3.8961 -1.6692 0>}]
 dyn_hair_01_l [{shapeType=Capsule, length=6, radius=3, shapeName=dyn_hair_01_l_capsule, center=<3.8961 -1.6692 0>}]
 dyn_hair_01_r [{shapeType=Capsule, length=6, radius=3, shapeName=dyn_hair_01_r_capsule, center=<-3.8961 1.6692 0>, rotation=<rot 0 0 -180>}]
 dyn_hair_02_c [{shapeType=Capsule, length=6, radius=3, shapeName=dyn_hair_02_c_capsule, center=<10.1097 -1.5039 0>}]
 dyn_hair_02_l [{shapeType=Capsule, length=6, radius=3, shapeName=dyn_hair_02_l_capsule, center=<7.2263 -1.6692 0>}]
 dyn_hair_02_r [{shapeType=Capsule, length=6, radius=3, shapeName=dyn_hair_02_r_capsule, center=<-7.2263 1.6692 0>, rotation=<rot 0 0 -180>}]
}
constraints [
 {
  bone1Name=base_hair_c
  bone2Name=deform_head
  swing1LimitDegrees=45
  swing1Motion=Free
  swing2LimitDegrees=45
  swing2Motion=Free
  twistLimitDegrees=45
  twistMotion=Free
 }
 {bone1Name=dyn_hair_01_c, bone2Name=base_hair_c, swing1LimitDegrees=20, swing2LimitDegrees=10, twistLimitDegrees=45, twistMotion=Locked}
 {bone1Name=dyn_hair_02_c, bone2Name=dyn_hair_01_c, swing1LimitDegrees=30, swing2LimitDegrees=10, twistLimitDegrees=45, twistMotion=Locked}
 {
  bone1Name=base_hair_l
  bone2Name=deform_head
  swing1LimitDegrees=45
  swing1Motion=Free
  swing2LimitDegrees=45
  swing2Motion=Free
  twistLimitDegrees=45
  twistMotion=Free
 }
 {bone1Name=dyn_hair_01_l, bone2Name=base_hair_l, swing1LimitDegrees=20, swing2LimitDegrees=10, twistLimitDegrees=45, twistMotion=Locked}
 {bone1Name=dyn_hair_02_l, bone2Name=dyn_hair_01_l, swing1LimitDegrees=35, swing2LimitDegrees=10, twistLimitDegrees=45, twistMotion=Locked}
 {
  bone1Name=base_hair_r
  bone2Name=deform_head
  swing1LimitDegrees=45
  swing1Motion=Free
  swing2LimitDegrees=45
  swing2Motion=Free
  twistLimitDegrees=45
  twistMotion=Free
 }
 {bone1Name=dyn_hair_01_r, bone2Name=base_hair_r, swing1LimitDegrees=20, swing2LimitDegrees=10, twistLimitDegrees=45, twistMotion=Locked}
 {bone1Name=dyn_hair_02_r, bone2Name=dyn_hair_01_r, swing1LimitDegrees=35, swing2LimitDegrees=10, twistLimitDegrees=45, twistMotion=Locked}
 {
  bone1Name=chest_back
  bone2Name=deform_spine_05
  swing1LimitDegrees=45
  swing1Motion=Free
  swing2LimitDegrees=45
  swing2Motion=Free
  twistLimitDegrees=45
  twistMotion=Free
 }
]
