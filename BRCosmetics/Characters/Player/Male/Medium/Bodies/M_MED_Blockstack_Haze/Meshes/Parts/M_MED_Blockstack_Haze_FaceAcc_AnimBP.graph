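# Anim Blueprint for the Blockstack Haze face accessory on the medium male body, handling its hair and chest-back bones.
# The copied parent pose runs through a Control Rig, additive bone offsets, head-relative RigidBody simulation, then hair offsets and Look Ats.
# EventGraph derives movement, targeting and melee state flags, many specifically for when rigid body is disabled, plus gravity and simulation settings from a parameters asset.
# Weapon changes set pistol flags used by the targeting states.
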AnimBlueprint /BRCosmetics/Characters/Player/Male/Medium/Bodies/M_MED_Blockstack_Haze/Meshes/Parts/M_MED_Blockstack_Haze_FaceAcc_AnimBP
target_skeleton=&/M_MED_Blockstack_Haze_FaceAcc_Skeleton
graphs [
 AnimGraph {nodes [
   n1 = Control Rig!(
    Source=Copy Pose From Mesh() {bUseAttachedParent=true}
    Alpha
    RBAN Enable:bool=Get bRigidBodyEnabled()
    Sprinting:bool=Get bIsSprinting()
    Targeting:bool=Get bIsTargeting()
    Crouching:bool=Get bIsCrouching()
    MeleeIDle:bool=Get MeleeIdleNoRBAN() -> (MeleeIdleNoRBAN:bool)
   ) -> (Pose) {controlRigAssetReference {controlRigAsset=&/M_MED_Blockstack_Haze_FaceAcc_CtrlRig}}
   n2 = "Transform (Modify) Bone - Bone: deform_spine_05"!(ComponentPose=Local To Component(LocalPose=@n1), bAlphaBoolEnabled:bool=True, Translation="0,2,0", Rotation="0,0,0", Scale="1,1,1") -> (Pose) {
    alphaInputType=Bool
    rotationMode=BMM_Additive
    rotationSpace=BCS_BoneSpace
    scaleMode=BMM_Additive
    scaleSpace=BCS_BoneSpace
    translationMode=BMM_Additive
    translationSpace=BCS_BoneSpace
    alphaBoolBlend {blendInTime=0.1, blendOutTime=0.1}
    boneToModify {boneName=deform_spine_05}
    translation {y=2}
   }
   n3 = "Transform (Modify) Bone - Bone: deform_spine_05"(ComponentPose=@n2, bAlphaBoolEnabled=True, Translation="0,1.5,0", Rotation="0,0,0", Scale="1,1,1") {
    alphaInputType=Bool
    rotationMode=BMM_Additive
    rotationSpace=BCS_BoneSpace
    scaleMode=BMM_Additive
    scaleSpace=BCS_BoneSpace
    translationMode=BMM_Additive
    translationSpace=BCS_BoneSpace
    alphaBoolBlend {blendInTime=0.1, blendOutTime=0.1}
    boneToModify {boneName=deform_spine_05}
    translation {y=1.5}
   }
   n4 = "Transform (Modify) Bone - Bone: chest_back"(ComponentPose=@n3, Alpha=1, Translation="0,-8,0", Rotation="0,0,0", Scale="1,1,1") {
    rotationMode=BMM_Additive
    rotationSpace=BCS_BoneSpace
    scaleMode=BMM_Additive
    scaleSpace=BCS_BoneSpace
    translationMode=BMM_Additive
    translationSpace=BCS_BoneSpace
    alphaBoolBlend {blendInTime=0.1, blendOutTime=0.1}
    alphaScaleBiasClamp {bInterpResult=true, interpSpeedDecreasing=0.5}
    boneToModify {boneName=chest_back}
    translation {y=-8}
   }
   n5 = "Transform (Modify) Bone - Bone: chest_back"!(ComponentPose=@n4, bAlphaBoolEnabled:bool=True, Translation="0,-8,0", Rotation="0,0,0", Scale="1,1,1") -> (Pose) {
    alphaInputType=Bool
    rotationMode=BMM_Additive
    rotationSpace=BCS_BoneSpace
    scaleMode=BMM_Additive
    scaleSpace=BCS_BoneSpace
    translationMode=BMM_Additive
    translationSpace=BCS_BoneSpace
    alphaBoolBlend {blendInTime=0.1, blendOutTime=0.1}
    alphaScaleBiasClamp {bInterpResult=true, interpSpeedDecreasing=3}
    boneToModify {boneName=chest_back}
    translation {y=-8}
   }
   n6 = RigidBody!(
    ComponentPose=@n5
    Alpha
    OverrideWorldGravity:Vector="0,0,-980"
    ExternalForce
    SimSpaceSettings:Sim Space Settings Structure=(WorldAlpha=1,VelocityScaleZ=1,DampingAlpha=1,MaxLinearVelocity=100,MaxAngularVelocity=15,MaxLinearAcceleration=200,MaxAngularAcceleration=50,ExternalLinearDragV=(X=0,Y=0,Z=0),ExternalLinearVelocity=(X=0,Y=0,Z=0),ExternalAngularVelocity=(X=0,Y=0,Z=0))
   ) -> (Pose) {
    bOverrideWorldGravity=true
    overridePhysicsAsset=&/M_MED_Blockstack_Haze_FaceAcc_Physics
    simulationSpace=BaseBoneSpace
    baseBoneRef {boneName=deform_head}
    componentAppliedLinearAccClamp=<0 0 0>
    overrideWorldGravity {z=-980}
    simSpaceSettings {maxAngularAcceleration=50, maxAngularVelocity=15, maxLinearAcceleration=200, maxLinearVelocity=100, worldAlpha=1}
   }
   n7 = "Transform (Modify) Bone - Bone: base_hair_general"(ComponentPose=@n6, Alpha=1, Translation="0,0,0", Rotation="0,-3,0", Scale="1,1,1") {
    rotationMode=BMM_Additive
    rotationSpace=BCS_BoneSpace
    scaleMode=BMM_Additive
    scaleSpace=BCS_BoneSpace
    translationMode=BMM_Additive
    translationSpace=BCS_BoneSpace
    alphaBoolBlend {blendInTime=0.1, blendOutTime=0.1}
    alphaScaleBiasClamp {bInterpResult=true, interpSpeedDecreasing=3}
    boneToModify {boneName=base_hair_general}
    rotation {yaw=-3}
   }
   n8 = "Transform (Modify) Bone - Bone: base_hair_general"!(ComponentPose=@n7, bAlphaBoolEnabled:bool=True, Translation="0,-5,0", Rotation="0,-12,0", Scale="1,1,1") -> (Pose) {
    alphaInputType=Bool
    rotationMode=BMM_Additive
    rotationSpace=BCS_BoneSpace
    scaleMode=BMM_Additive
    scaleSpace=BCS_BoneSpace
    translationMode=BMM_Additive
    translationSpace=BCS_BoneSpace
    alphaBoolBlend {blendInTime=0.2, blendOutTime=1}
    alphaScaleBiasClamp {bInterpResult=true, interpSpeedDecreasing=3}
    boneToModify {boneName=base_hair_general}
    rotation {yaw=-12}
    translation {y=-5}
   }
   n9 = "Transform (Modify) Bone - Bone: base_hair_general"(ComponentPose=@n8, bAlphaBoolEnabled=True, Translation="0,0,0", Rotation="0.330,-8.771,0.057", Scale="1,1,1") {
    alphaInputType=Bool
    rotationMode=BMM_Additive
    rotationSpace=BCS_BoneSpace
    scaleMode=BMM_Additive
    scaleSpace=BCS_BoneSpace
    translationMode=BMM_Additive
    translationSpace=BCS_BoneSpace
    alphaBoolBlend {blendInTime=0.2, blendOption=Cubic, blendOutTime=0.8}
    alphaScaleBiasClamp {bInterpResult=true, interpSpeedDecreasing=3}
    boneToModify {boneName=base_hair_general}
    rotation=<rot 0.33 -8.771 0.057>
   }
   n10 = "Transform (Modify) Bone - Bone: base_hair_general"(ComponentPose=@n9, bAlphaBoolEnabled=True, Translation="0,-5,0", Rotation="0,-15,0", Scale="1,1,1") {
    alphaInputType=Bool
    rotationMode=BMM_Additive
    rotationSpace=BCS_BoneSpace
    scaleMode=BMM_Additive
    scaleSpace=BCS_BoneSpace
    translationMode=BMM_Additive
    translationSpace=BCS_BoneSpace
    alphaBoolBlend {blendInTime=0.2, blendOption=Cubic, blendOutTime=1}
    alphaScaleBiasClamp {bInterpResult=true, interpSpeedDecreasing=3}
    boneToModify {boneName=base_hair_general}
    rotation {yaw=-15}
    translation {y=-5}
   }
   n11 = "Transform (Modify) Bone - Bone: base_hair_general"(ComponentPose=@n10, bAlphaBoolEnabled=True, Translation="-2,0,0", Rotation="0,-10,0", Scale="1,1,1") {
    alphaInputType=Bool
    rotationMode=BMM_Additive
    rotationSpace=BCS_BoneSpace
    scaleMode=BMM_Additive
    scaleSpace=BCS_BoneSpace
    translationMode=BMM_Additive
    translationSpace=BCS_BoneSpace
    alphaBoolBlend {blendInTime=0.2, blendOption=Cubic, blendOutTime=0.7}
    alphaScaleBiasClamp {bInterpResult=true, interpSpeedDecreasing=3}
    boneToModify {boneName=base_hair_general}
    rotation {yaw=-10}
    translation {x=-2}
   }
   n12 = "Transform (Modify) Bone - Bone: base_hair_r"!(ComponentPose=@n11, bAlphaBoolEnabled:bool=True, Translation="0,0,0", Rotation="-0.000,-8.074,-0.000", Scale="1,1,1") -> (Pose) {
    alphaInputType=Bool
    rotationMode=BMM_Additive
    rotationSpace=BCS_BoneSpace
    scaleMode=BMM_Additive
    scaleSpace=BCS_BoneSpace
    translationMode=BMM_Additive
    translationSpace=BCS_BoneSpace
    alphaBoolBlend {blendInTime=0.2, blendOption=Cubic, blendOutTime=0.7}
    alphaScaleBiasClamp {bInterpResult=true, interpSpeedDecreasing=3}
    boneToModify {boneName=base_hair_r}
    rotation {yaw=-8.074}
   }
   n13 = "Transform (Modify) Bone - Bone: base_hair_general"(ComponentPose=@n12, bAlphaBoolEnabled=True, Translation="0,0,0", Rotation="0,-10,0", Scale="1,1,1") {
    alphaInputType=Bool
    rotationMode=BMM_Additive
    rotationSpace=BCS_BoneSpace
    scaleMode=BMM_Additive
    scaleSpace=BCS_BoneSpace
    translationMode=BMM_Additive
    translationSpace=BCS_BoneSpace
    alphaBoolBlend {blendInTime=0.2, blendOption=Cubic, blendOutTime=0.7}
    alphaScaleBiasClamp {bInterpResult=true, interpSpeedDecreasing=3}
    boneToModify {boneName=base_hair_general}
    rotation {yaw=-10}
   }
   n14 = "Transform (Modify) Bone - Bone: base_hair_general"(ComponentPose=@n13, bAlphaBoolEnabled=True, Translation="0,0,0", Rotation="0,-10,0", Scale="1,1,1") {
    alphaInputType=Bool
    rotationMode=BMM_Additive
    rotationSpace=BCS_BoneSpace
    scaleMode=BMM_Additive
    scaleSpace=BCS_BoneSpace
    translationMode=BMM_Additive
    translationSpace=BCS_BoneSpace
    alphaBoolBlend {blendInTime=0.2, blendOption=Cubic, blendOutTime=0.7}
    alphaScaleBiasClamp {bInterpResult=true, interpSpeedDecreasing=3}
    boneToModify {boneName=base_hair_general}
    rotation {yaw=-10}
   }
   n15 = Look At - Bone: base_hair_general!(ComponentPose=@n14, bAlphaBoolEnabled:bool=True) -> (Pose) {
    alphaInputType=Bool
    alphaBoolBlend {blendInTime=0.2, blendOutTime=0.8}
    boneToModify {boneName=base_hair_general}
    lookAtLocation=<-58.1066 -52.3365 -5.0777>
    lookAtTarget {boneReference {boneName=spine_05}}
    lookAt_Axis {axis=<1 0>}
   }
   n16 = Look At - Bone: base_hair_general(ComponentPose=@n15, bAlphaBoolEnabled=True) {
    alphaInputType=Bool
    alphaBoolBlend {blendInTime=0.2, blendOutTime=0.8}
    boneToModify {boneName=base_hair_general}
    lookAtLocation=<-49.2181 -48.1653 -1.5183>
    lookAtTarget {boneReference {boneName=spine_05}}
    lookAt_Axis {axis=<1 0>}
   }
   Output Pose(Result=Component To Local(ComponentPose=@n16)) {layerGroup=None}
  ]}
 EventGraph {nodes [
   Event BlueprintUpdateAnimation()
   > n1 = Sequence!(execute) -> (then_0, then_1, then_2:Exec)
   > n2 = "Gravity Override State Function (Parameter Asset) (Nativized)"(
    CharacterPartAnimInstance=Self-Reference()
    GameTimeInSeconds=GetGameTimeInSeconds()
    DebugJointName=spine_05
    Parameters=&/M_MED_Blockstack_Haze_FaceAcc_Parameters
   )
   > Set Gravity Override(Gravity Override=@n2)
   > Set RBAN Settings Out(RBAN Settings Out=@n2.OutRigidBodyAnimNodeParameters)
   n3 = Get bRigidBodyEnabled()
   Set TargetingWithPistolNoRban(
    execute=@n1.then_2
    TargetingWithPistolNoRban:bool=AND Boolean!(A=NOT Boolean(A=@n3), B=Get Pistol() -> (Pistol:bool), C:bool=Get bIsTargeting()) -> (ReturnValue)
   ) -> (Output_Get:bool)
   n4 = Get bRigidBodyEnabled()
   > Set TargetingNoPistolNoRban(TargetingNoPistolNoRban:bool=AND Boolean(A=NOT Boolean(A=@n4), B=Get NoPistol() -> (NoPistol:bool), C=Get bIsTargeting())) -> (Output_Get:bool)
   n5 = Get bRigidBodyEnabled()
   > Set Melee in Crouching NoRban(Melee in Crouching NoRban:bool=AND Boolean(A=Get bIsCrouching(), B=Get bIsMoving2D(), C=NOT Boolean(A=@n5))) -> (Output_Get:bool)
   n6 = Get bRigidBodyEnabled()
   > Set SkydivingNO Rban(SkydivingNO Rban:bool=AND Boolean!(A=Get bIsSkyDiving(), B=NOT Boolean(A=@n6)) -> (ReturnValue)) -> (Output_Get:bool)
   n7 = Get bRigidBodyEnabled()
   > Set OpenParachuteNoRban(OpenParachuteNoRban:bool=AND Boolean(A=Get bIsParachuteOpened(), B=NOT Boolean(A=@n7))) -> (Output_Get:bool)
   n8 = AND Boolean(A=Get bIsCrouching(), B=Get bIsMoving2D())
   n9 = AND Boolean(A=OR Boolean(A=Get bIsJumping(), B=Get bIsFalling()), B=NOT Boolean(A=Get bIsSkyDiving()))
   n10 = Get bIsSprinting()
   n11 = Get bIsTacticalSprinting()
   Set JumpingAndTacticalSprinting(
    execute=@n1.then_1
    JumpingAndTacticalSprinting:double="To Float (Boolean)"(InBool=OR Boolean!(A=@n11, B=@n9, C:bool=@n8, D:bool=@n10) -> (ReturnValue))
   ) -> (Output_Get:double)
   n12 = NOT Boolean(A=Get bIsMoving2D())
   n13 = Get bIsCrouching()
   > Set CrouchingIdle(CrouchingIdle:double="To Float (Boolean)"(InBool=AND Boolean(A=@n13, B=@n12))) -> (Output_Get:double)
   n14 = Get bIsSkyDiving()
   n15 = Get bRigidBodyEnabled()
   n16 = Get StandingMoving2dNoRBAN() -> (StandingMoving2dNoRBAN:bool)
   n17 = Get bIsJumping()
   n18 = Get bIsFalling()
   > Set JumpingCompleteNoRBAN(JumpingCompleteNoRBAN:bool=AND Boolean!(A=OR Boolean!(A=@n17, B=@n18) -> (ReturnValue), B=NOT Boolean(A=@n14), C:bool=NOT Boolean(A=@n15), D:bool=NOT Boolean(A=@n16)) -> (ReturnValue)) -> (Output_Get:bool)
   n19 = NOT Boolean(A=Get bIsCrouching())
   n20 = NOT Boolean(A=Get bRigidBodyEnabled())
   n21 = NOT Boolean(A=Get bIsTargeting())
   n22 = Get bIsMoving2D()
   > Set StandingMoving2dNoRBAN(StandingMoving2dNoRBAN:bool=OR Boolean(A=AND Boolean(A=@n22, B=@n20, C=@n19, D=@n21), B=Get bIsPlayingEmote())) -> (Output_Get:bool)
   n23 = Get bRigidBodyEnabled()
   > Set CrouchingNoRBAN(CrouchingNoRBAN:bool=AND Boolean!(A=Get bIsCrouching(), B=NOT Boolean(A=@n23)) -> (ReturnValue)) -> (Output_Get:bool)
   n24 = Get bIsMoving2D()
   n25 = Get bRigidBodyEnabled()
   > Set MeleeIdleNoRBAN(MeleeIdleNoRBAN:bool=AND Boolean!(A=Get bIsPlayingMeleeAnim(), B=NOT Boolean(A=@n25), C:bool=NOT Boolean(A=@n24)) -> (ReturnValue)) -> (Output_Get:bool)
   Event OnWeaponChanged()
   > n26 = Cast To FortPlayerAnimInstance(Object=GetAnimInstance(self=GetOwnerSkeletalMeshComponent()))
   > Set Pistol(Pistol:bool=false) -> (Output_Get:bool)
   > Set NoPistol(NoPistol:bool=true) -> (Output_Get:bool)
   > n27 = Switch on EFortWeaponCoreAnimation!(execute, Selection:EFortWeaponCoreAnimation Enum=Get WeaponCoreAnim(self=@n26)) -> (
    Melee:Exec
    Pistol:Exec
    Shotgun:Exec
    PaperBlueprint:Exec
    Rifle:Exec
    MeleeOneHand:Exec
    MachinePistol:Exec
    RocketLauncher:Exec
    GrenadeLauncher:Exec
    GoingCommando:Exec
    AssaultRifle:Exec
    TacticalShotgun:Exec
    SniperRifle:Exec
    TrapPlacement:Exec
    ShoulderLauncher:Exec
    AbilityDecoTool:Exec
    Crossbow:Exec
    C4:Exec
    RemoteControl:Exec
    DualWield:Exec
    AR_BullPup:Exec
    AR_ForwardGrip:Exec
    MedPackPaddles:Exec
    SMG_P90:Exec
    AR_DrumGun:Exec
    Consumable_Small:Exec
    Consumable_Large:Exec
    Balloon:Exec
    MountedTurret:Exec
    CreativeTool:Exec
    ExplosiveBow:Exec
    AshtonIndigo:Exec
    AshtonChicago:Exec
    MeleeDualWield:Exec
    Unarmed:Exec
   )
   Set Pistol(execute=@n27.Pistol, Pistol=true)
   > Set NoPistol(NoPistol=false)
  ]}
]
variables [
 CrouchingIdle {type=number}
 CrouchingNoRBAN {type=boolean}
 Gravity Override {type=Vector}
 JumpingAndTacticalSprinting {type=number}
 JumpingCompleteNoRBAN {type=boolean}
 Melee in Crouching NoRban {type=boolean}
 MeleeIdleNoRBAN {type=boolean}
 NoPistol {type=boolean}
 OpenParachuteNoRban {type=boolean}
 Pistol {type=boolean}
 RBAN Settings Out {type=FortRigidBodyAnimNodeParameters}
 SkydivingNO Rban {type=boolean}
 StandingMoving2dNoRBAN {type=boolean}
 TargetingNoPistolNoRban {type=boolean}
 TargetingWithPistolNoRban {type=boolean}
]
